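# Material function that produces a screen-space, time-varying dither around an Alpha Threshold for use with Temporal AA, as in masked translucency.
# Pixel position offset by the temporal sample index feeds a Custom node computing a five-step (x + 2y) mod 5 pattern.
# A pixel-aligned 256x256 noise texture, blue noise per the asset name, is scaled by Random and added to that pattern.
# The sum is divided by 6, added to the threshold and shifted down by 0.5.
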MaterialFunction /Game/TimeOfDay/CloudsAndStorm/Materials/MaterialFunctions/DitherTemporalAABlue
graph {
 properties {
  bExposeToLibrary=true
  description="Stipple pattern/Dither  in screen space and time to work well with TemporalAA for things like masked translucency or anisotropic materials"
 }
 nodes [
  n1 = FunctionInput() {bUsePreviewValueAsDefault=true, inputName=Alpha Threshold, inputType=FunctionInput_Scalar, sortPriority=-1, previewValue=<0.5 0 0 1>}
  n2 = ViewProperty() {property=MEVP_TemporalSampleIndex}
  MaterialExpressionAdd_2 = Add(A=ScreenPosition().PixelPosition, B=AppendVector(A=@n2.Property, B=@n2.Property))
  n3 = Custom!(p=@MaterialExpressionAdd_2) {
   code="Mod( ((uint)(p.x) + 2 * (uint)(p.y)) ,5 )"
   outputType=CMOT_Float1
   inputs [{inputName=p, input {expression=&self:MaterialExpressionAdd_2}}]
  }
  n4 = TextureSample(UVs=MaterialFunctionCall!(TextureResolution=Constant2Vector() {g=256, r=256}) {materialFunction=/Engine/Functions/Engine_MaterialFunctions02/Texturing/ScreenAlignedPixelToPixelUVs}.Result) {samplerType=SAMPLERTYPE_LinearGrayscale, texture=/Game/TimeOfDay/CloudsAndStorm/Textures/LDR_LLL1_0}
  n5 = FunctionInput() {
   bUsePreviewValueAsDefault=true
   description="0=no random 1:random(default)"
   inputName=Random
   inputType=FunctionInput_Scalar
   previewValue=<1 0 0 1>
  }
  n6 = Divide(A=Add(A=@n3, B=Multiply(A=@n4.R, B=@n5)) {constB=0}) {constB=6}
  FunctionOutput(None=Add(A=Add(A=@n1, B=@n6) {constB=0.5}) {constB=-0.5})
 ]
}
