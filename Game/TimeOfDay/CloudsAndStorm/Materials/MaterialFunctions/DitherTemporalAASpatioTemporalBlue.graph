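# Material function that produces screen-space, time-varying blue-noise dither suited to Temporal AA, for uses like masked translucency.
# It builds 8-frame and 64-frame noise variants from screen-aligned pixel UVs and View.StateFrameIndex, with Tile1Dto2D tile lookups, plus a fake temporal variant that offsets noise by the golden ratio.
# Each variant is scaled by Random and biased around Alpha Threshold, with both raw and thresholded results output.
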MaterialFunction /Game/TimeOfDay/CloudsAndStorm/Materials/MaterialFunctions/DitherTemporalAASpatioTemporalBlue
graph {
 properties {
  bExposeToLibrary=true
  description="Stipple pattern/Dither  in screen space and time to work well with TemporalAA for things like masked translucency or anisotropic materials"
 }
 nodes [
  n1 = FunctionInput() {bUsePreviewValueAsDefault=true, inputName=Alpha Threshold, inputType=FunctionInput_Scalar, sortPriority=-1, previewValue=<0.5 0 0 1>}
  n2 = FunctionInput() {
   bUsePreviewValueAsDefault=true
   description="0=no random 1:random(default)"
   inputName=Random
   inputType=FunctionInput_Scalar
   previewValue=<1 0 0 1>
  }
  Add(A=Add(A=@n1, B=Multiply(B=@n2)) {constB=0.5}) {constB=-0.5}
  n3 = Frac(None=Add(B=Custom!(None) {code="return (View.StateFrameIndex * 1.61803398875) ;", outputType=CMOT_Float1}))
  Add(A=Add(A=Multiply(A=@n3, B=@n2), B=@n1)) {constB=-0.5}
  MaterialExpressionConstant2Vector_3 = Constant2Vector() {g=8, r=8}
  MaterialExpressionRound_0 = Round(None=Custom() {code="return View.StateFrameIndex % 64;", outputType=CMOT_Float1})
  n4 = Custom!(xysize=@MaterialExpressionConstant2Vector_3, zframe=@MaterialExpressionRound_0) {
   code="// The below expression will get compiled\r\n// into the output of this node\r\nreturn Tile1Dto2D(xysize.x, zframe) / xysize;\r\n"
   outputType=CMOT_Float2
   showCode=true
   inputs [
    {
     inputName=xysize
     input {
      expression=&self:MaterialExpressionConstant2Vector_3
      mask=1
      maskG=1
      maskR=1
     }
    }
    {
     inputName=zframe
     input {expression=&self:MaterialExpressionRound_0}
    }
   ]
  }
  Add(A=Divide(A=NamedRerouteUsage(), B=@MaterialExpressionConstant2Vector_3) {constB=8}, B=@n4)
  Add(A=Add(A=@n1, B=Multiply(B=@n2)) {constB=0.5}) {constB=-0.5}
  MaterialExpressionConstant2Vector_1 = Constant2Vector() {g=8, r=8}
  MaterialExpressionRound_1 = Round(None=Custom!(None) {code="return View.StateFrameIndexMod8;", outputType=CMOT_Float1})
  n5 = Custom!(xysize=@MaterialExpressionConstant2Vector_1, zframe=@MaterialExpressionRound_1) {
   code="// The below expression will get compiled\r\n// into the output of this node\r\nreturn Tile1Dto2D(xysize.x, zframe) / xysize;\r\n"
   outputType=CMOT_Float2
   showCode=true
   inputs [
    {
     inputName=xysize
     input {
      expression=&self:MaterialExpressionConstant2Vector_1
      mask=1
      maskG=1
      maskR=1
     }
    }
    {
     inputName=zframe
     input {expression=&self:MaterialExpressionRound_1}
    }
   ]
  }
  Add(A=Divide(A=NamedRerouteUsage(), B=@MaterialExpressionConstant2Vector_1) {constB=8}, B=@n5)
  n6 = MaterialFunctionCall!(TextureResolution=Constant2Vector() {g=128, r=128}) {materialFunction=/Engine/Functions/Engine_MaterialFunctions02/Texturing/ScreenAlignedPixelToPixelUVs}
  AppendVector(A=@n6.Result, B=Custom!(None) {code="return View.StateFrameIndexMod8;", outputType=CMOT_Float1})
  AppendVector(A=@n6.Result, B=Custom() {code="return View.StateFrameIndex % 64;", outputType=CMOT_Float1})
  Constant()
  Constant()
  Constant()
  Constant() {r=63}
  n7 = NamedRerouteUsage()
  Frac(None=Add(A=@n7, B=Custom() {code="return (View.StateFrameIndex * 1.61803398875) ;", outputType=CMOT_Float1}))
  n8 = NamedRerouteUsage()
  n9 = NamedRerouteUsage()
  n10 = Multiply(A=@n9, B=NamedRerouteUsage())
  FunctionOutput(None=Add(A=Add(A=@n8, B=@n10) {constB=0.5}) {constB=-0.5})
  n11 = NamedRerouteUsage()
  FunctionOutput(None=@n11) {outputName=RawNoise64Frames}
  FunctionOutput(None=@n9) {outputName=RawNoise8Frames}
  n12 = NamedRerouteUsage()
  n13 = Multiply(A=@n11, B=NamedRerouteUsage())
  FunctionOutput(None=Add(A=Add(A=@n12, B=@n13) {constB=0.5}) {constB=-0.5}) {outputName=Result64}
  FunctionOutput(None=@n7) {outputName=RawFakeTemporalGoldenRatioBlueNoise}
  n14 = NamedRerouteUsage()
  n15 = Multiply(A=@n7, B=NamedRerouteUsage())
  FunctionOutput(None=Add(A=Add(A=@n14, B=@n15) {constB=0.5}) {constB=-0.5}) {outputName=ResultFakeTemporalGoldenRatioBlueNoise}
  Multiply()
  NamedRerouteDeclaration(None=@n6.Result) {name=ScreenAlignedUVs, nodeColor=<rgba 0 0.1562 1 1>}
  NamedRerouteDeclaration(None=@n1) {name=Alpha Threshold, nodeColor=<rgba 0.4766 1 0 1>}
  NamedRerouteDeclaration(None=@n2) {name=Random, nodeColor=<rgba 1 0 0.7969 1>}
  NamedRerouteDeclaration(None=ScalarBlueNoise()) {name=BN, nodeColor=<rgba 1 0.7969 0 1>}
  NamedRerouteUsage()
  TextureCoordinate()
 ]
}
