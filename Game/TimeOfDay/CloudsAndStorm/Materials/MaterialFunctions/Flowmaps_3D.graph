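# Material function that performs flowmap blending on a volume texture, stored among the clouds and storm time-of-day materials.
# It offsets UVW by velocity times two sawtooth time phases half a cycle apart, with time overridable, then samples both at a chosen mip.
# A phase-shifted sine crossfades the samples to hide each reset.
# It outputs blended RGB and RGBA plus both distorted coordinate sets.
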MaterialFunction /Game/TimeOfDay/CloudsAndStorm/Materials/MaterialFunctions/Flowmaps_3D
graph {
 properties {description=Performs Flowmap blending using a Volume Texture., libraryCategoriesText [Texturing]}
 nodes [
  n1 = FunctionInput(Preview=Time()) {bUsePreviewValueAsDefault=true, inputName=Override Time, inputType=FunctionInput_Scalar, sortPriority=3}
  n2 = Subtract(A=@n1) {constB=0.5}
  n3 = FunctionInput() {bUsePreviewValueAsDefault=true, inputName=Velocity, sortPriority=2}
  n4 = FunctionInput(Preview=WorldPosition().XYZ) {bUsePreviewValueAsDefault=true, inputName=UVW, sortPriority=1}
  n5 = Subtract(A=@n4, B=Multiply(A=@n3, B=Frac(None=@n2)))
  n6 = FunctionInput() {bUsePreviewValueAsDefault=true, inputName=Mip Level, inputType=FunctionInput_Scalar, sortPriority=4}
  n7 = FunctionInput(Preview=TextureObject() {samplerType=SAMPLERTYPE_Masks, texture=/Engine/EngineSky/VolumetricClouds/T_VolumeNoiseShape128}) {bUsePreviewValueAsDefault=true, inputName=Volume Texture, inputType=FunctionInput_VolumeTexture}
  n8 = TextureSample!(UVs=@n5, Tex=@n7, Level=@n6, Apply View MipBias) {mipValueMode=TMVM_MipLevel}
  n9 = Subtract(A=@n4, B=Multiply(A=@n3, B=Frac(None=@n1)))
  n10 = TextureSample(UVs=@n9, Tex=@n7, Level=@n6) {mipValueMode=TMVM_MipLevel}
  n11 = ConstantBiasScale(None=Sine(None=Add(A=@n1) {constB=-0.25}))
  FunctionOutput(None=LinearInterpolate(A=@n8.RGB, B=@n10.RGB, Alpha=@n11)) {outputName=Result RGB}
  FunctionOutput(None=LinearInterpolate(A=@n8.RGBA, B=@n10.RGBA, Alpha=@n11)) {outputName=Result RGBA, sortPriority=1}
  FunctionOutput(None=@n5) {outputName=UVW_A}
  FunctionOutput(None=@n9) {outputName=UVW_B}
 ]
}
