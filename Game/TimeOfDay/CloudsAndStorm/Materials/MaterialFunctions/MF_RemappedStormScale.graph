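# MF_RemappedStormScale is a material function that remaps safe zone storm scale for the cloud and storm materials.
# It reads SafeZoneScale from MPC_SafeZoneClouds and normalizes it two ways: over a short window ending at a start-phase threshold, and from zero to that threshold.
# The RemappedStorm collection parameter blends CloudStartingBias and raw SafeZoneScale toward those saturated values, producing the StartingBias and ZoneScale outputs.
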MaterialFunction /Game/TimeOfDay/CloudsAndStorm/Materials/MaterialFunctions/MF_RemappedStormScale
graph {nodes [
  n1 = ScalarParameter() {defaultValue=0.47, parameterName=StormScaleBasedStartPhase}
  n2 = InverseLinearInterpolate(
   A=Subtract(A=@n1, B=ScalarParameter() {defaultValue=0.03, parameterName=StormScaleBasedStartDuration})
   B=@n1
   Value=CollectionParameter() {collection=/Game/Athena/SafeZone/MPC_SafeZoneClouds, parameterName=SafeZoneScale}
  ) {constA=0.4, constB=0.5}
  FunctionOutput(None=LinearInterpolate(
    A=CollectionParameter() {collection=&MPC_SafeZoneClouds, parameterName=CloudStartingBias}
    B=Saturate(None=@n2)
    Alpha=CollectionParameter() {collection=&MPC_SafeZoneClouds, parameterName=RemappedStorm}
   )) {outputName=StartingBias}
  n3 = InverseLinearInterpolate(B=@n1, Value=CollectionParameter() {collection=&MPC_SafeZoneClouds, parameterName=SafeZoneScale}) {constB=0.5}
  FunctionOutput(None=LinearInterpolate(
    A=CollectionParameter() {collection=&MPC_SafeZoneClouds, parameterName=SafeZoneScale}
    B=Saturate(None=@n3)
    Alpha=CollectionParameter() {collection=&MPC_SafeZoneClouds, parameterName=RemappedStorm}
   )) {outputName=ZoneScale, sortPriority=1}
 ]}
