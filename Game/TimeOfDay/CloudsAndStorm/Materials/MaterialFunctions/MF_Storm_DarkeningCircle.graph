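# Material function that darkens an input color in a horizontal ring of the sky for storms, tinting it by multiplying with a Storm Color.
# The ring comes from camera vector Z offset by Height, with abs, scale, saturate and repeated squaring shaping a soft falloff around that elevation.
# MF_RemappedStormScale's StartingBias output is doubled, clamped and inverted to scale the ring, so the darkening fades as that value rises.
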MaterialFunction /Game/TimeOfDay/CloudsAndStorm/Materials/MaterialFunctions/MF_Storm_DarkeningCircle
graph {nodes [
  n1 = FunctionInput() {inputName=Color}
  n2 = FunctionInput(Preview=Constant3Vector() {constant=<rgba 0.0237 0.0457 0.1198 1>}) {bUsePreviewValueAsDefault=true, inputName=Storm Color, sortPriority=1}
  n3 = FunctionInput() {bUsePreviewValueAsDefault=true, inputName=Height, inputType=FunctionInput_Scalar, sortPriority=2, previewValue=<0.08 0.08 0.08 1>}
  n4 = Add(A=ComponentMask(None=CameraVectorWS()) {b=true}, B=@n3) {constB=0.08}
  n5 = Saturate(None=Multiply(A=Abs(None=@n4)) {constB=6})
  n6 = OneMinus(None=Multiply(A=@n5, B=@n5))
  n7 = Multiply(A=@n6, B=@n6)
  n8 = MaterialFunctionCall() {materialFunction=&/MF_RemappedStormScale}
  n9 = OneMinus(None=Saturate(None=Multiply(A=@n8.StartingBias) {constB=2}))
  FunctionOutput(None=LinearInterpolate(A=@n1, B=Multiply(A=@n1, B=@n2), Alpha=Multiply(A=@n7, B=@n9)) {constB=0})
 ]}
