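# Material function in the clouds and storm set that remaps world position into scaled 2D XY coordinates.
# It subtracts a constant half unit from world position, then multiplies by the reciprocal of the WorldScale input, which has a very large default.
# The output keeps only red and green, so height is dropped.
# The offset is applied before scaling, so it is in world units rather than normalized space.
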MaterialFunction /Game/TimeOfDay/CloudsAndStorm/Materials/MaterialFunctions/MF_WorldRemap
graph {nodes [
  n1 = Subtract(A=WorldPosition().XYZ) {constB=0.5}
  n2 = Divide(
   A=Constant() {r=1}
   B=FunctionInput(Preview=Constant() {r=6000000}) {bUsePreviewValueAsDefault=true, inputName=WorldScale, inputType=FunctionInput_Scalar}
  )
  FunctionOutput(None=ComponentMask(None=Multiply(A=@n1, B=@n2)) {g=true, r=true})
 ]}
